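# Translucent UI material for the Pilgrim minimap, built around a captured map image.
# A single texture sample reads the MapTexture object parameter, which defaults to Pilgrim_MapCapture and can be overridden by material instances.
# The parameter uses the Normal sampler type rather than a color one.
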Material /SparksMaps/UI/Materials/M_MiniMapPilgrim
graph {
 properties {blendMode=BLEND_Translucent, materialDomain=MD_UI, pixelDepthOffsetMode=PDOM_Legacy, usedShadingModels=Default Lit}
 nodes [TextureSample(Tex=TextureObjectParameter() {parameterName=MapTexture, samplerType=SAMPLERTYPE_Normal, texture=&/Pilgrim_MapCapture})]
}
